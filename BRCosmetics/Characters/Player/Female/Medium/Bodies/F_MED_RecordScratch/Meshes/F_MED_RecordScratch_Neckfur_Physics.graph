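# Physics asset for the neck fur mesh of the RecordScratch female medium body cosmetic, defining simulation bodies for its dynamic fur bones.
# Each back and side fur bone gets a small offset capsule, while spine_05 carries a larger capsule for the upper torso.
# Every fur body is constrained directly to spine_05 rather than chained, with identical limits allowing moderate swing and very little twist.
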PhysicsAsset /BRCosmetics/Characters/Player/Female/Medium/Bodies/F_MED_RecordScratch/Meshes/F_MED_RecordScratch_Neckfur_Physics
bodies {
 dyn_fur_a_bck [{shapeType=Capsule, length=1, radius=1, center=<2 0 0>, rotation=<rot 90 0 0>}]
 dyn_fur_a_bck1 [{shapeType=Capsule, length=1, radius=1, center=<2 0 0>, rotation=<rot 90 0 0>}]
 dyn_fur_b_bck [{shapeType=Capsule, length=1, radius=1, center=<2 0 0>, rotation=<rot 90 0 0>}]
 dyn_fur_bck_Top_l [{shapeType=Capsule, length=1, radius=1, center=<2 0 0>, rotation=<rot 90 0 0>}]
 dyn_fur_bck_Top_m [{shapeType=Capsule, length=1, radius=1, center=<2 0 0>, rotation=<rot 90 0 0>}]
 dyn_fur_bck_Top_r [{shapeType=Capsule, length=1, radius=1, center=<2 0 0>, rotation=<rot 90 0 0>}]
 dyn_fur_bck_l1 [{shapeType=Capsule, length=1, radius=1, center=<3 0 0>, rotation=<rot 90 0 0>}]
 dyn_fur_bck_m [{shapeType=Capsule, length=1, radius=1, center=<2 0 0>, rotation=<rot 90 0 0>}]
 dyn_fur_d_bck [{shapeType=Capsule, length=1, radius=1, center=<3 0 0>, rotation=<rot 90 0 0>}]
 dyn_fur_e_bck [{shapeType=Capsule, length=1, radius=1, center=<2 0 0>, rotation=<rot 90 0 0>}]
 dyn_fur_f_bck [{shapeType=Capsule, length=1, radius=1, center=<3 0 0>, rotation=<rot 90 0 0>}]
 dyn_fur_side_a_l [{shapeType=Capsule, length=1, radius=1, center=<3 0 0>, rotation=<rot 90 0 0>}]
 dyn_fur_side_a_r [{shapeType=Capsule, length=1, radius=1, center=<3 0 0>, rotation=<rot 90 0 0>}]
 dyn_fur_side_a_r1 [{shapeType=Capsule, length=1, radius=1, center=<3 0 0>, rotation=<rot 90 0 0>}]
 dyn_fur_side_a_r2 [{shapeType=Capsule, length=1, radius=1, center=<3 0 0>, rotation=<rot 90 0 0>}]
 dyn_fur_side_b_l [{shapeType=Capsule, length=1, radius=1, center=<3 0 0>, rotation=<rot 90 0 0>}]
 dyn_fur_side_b_r [{shapeType=Capsule, length=1, radius=1, center=<3 0 0>, rotation=<rot 90 0 0>}]
 dyn_fur_side_c_l [{shapeType=Capsule, length=1, radius=1, center=<3 0 0>, rotation=<rot 90 0 0>}]
 spine_05 [{shapeType=Capsule, length=12.5104, radius=4.4235, center=<5.3364 -3.1289 0>}]
}
constraints [
 {bone1Name=dyn_fur_side_c_l, bone2Name=spine_05, swing1LimitDegrees=45, swing2LimitDegrees=45, twistLimitDegrees=5}
 {bone1Name=dyn_fur_side_b_l, bone2Name=spine_05, swing1LimitDegrees=45, swing2LimitDegrees=45, twistLimitDegrees=5}
 {bone1Name=dyn_fur_side_a_l, bone2Name=spine_05, swing1LimitDegrees=45, swing2LimitDegrees=45, twistLimitDegrees=5}
 {bone1Name=dyn_fur_bck_l1, bone2Name=spine_05, swing1LimitDegrees=45, swing2LimitDegrees=45, twistLimitDegrees=5}
 {bone1Name=dyn_fur_d_bck, bone2Name=spine_05, swing1LimitDegrees=45, swing2LimitDegrees=45, twistLimitDegrees=5}
 {bone1Name=dyn_fur_f_bck, bone2Name=spine_05, swing1LimitDegrees=45, swing2LimitDegrees=45, twistLimitDegrees=5}
 {bone1Name=dyn_fur_side_a_r, bone2Name=spine_05, swing1LimitDegrees=45, swing2LimitDegrees=45, twistLimitDegrees=5}
 {bone1Name=dyn_fur_side_a_r1, bone2Name=spine_05, swing1LimitDegrees=45, swing2LimitDegrees=45, twistLimitDegrees=5}
 {bone1Name=dyn_fur_side_a_r2, bone2Name=spine_05, swing1LimitDegrees=45, swing2LimitDegrees=45, twistLimitDegrees=5}
 {bone1Name=dyn_fur_side_b_r, bone2Name=spine_05, swing1LimitDegrees=45, swing2LimitDegrees=45, twistLimitDegrees=5}
 {bone1Name=dyn_fur_a_bck, bone2Name=spine_05, swing1LimitDegrees=45, swing2LimitDegrees=45, twistLimitDegrees=5}
 {bone1Name=dyn_fur_e_bck, bone2Name=spine_05, swing1LimitDegrees=45, swing2LimitDegrees=45, twistLimitDegrees=5}
 {bone1Name=dyn_fur_bck_Top_r, bone2Name=spine_05, swing1LimitDegrees=45, swing2LimitDegrees=45, twistLimitDegrees=5}
 {bone1Name=dyn_fur_bck_m, bone2Name=spine_05, swing1LimitDegrees=45, swing2LimitDegrees=45, twistLimitDegrees=5}
 {bone1Name=dyn_fur_bck_Top_m, bone2Name=spine_05, swing1LimitDegrees=45, swing2LimitDegrees=45, twistLimitDegrees=5}
 {bone1Name=dyn_fur_bck_Top_l, bone2Name=spine_05, swing1LimitDegrees=45, swing2LimitDegrees=45, twistLimitDegrees=5}
 {bone1Name=dyn_fur_a_bck1, bone2Name=spine_05, swing1LimitDegrees=45, swing2LimitDegrees=45, twistLimitDegrees=5}
 {bone1Name=dyn_fur_b_bck, bone2Name=spine_05, swing1LimitDegrees=45, swing2LimitDegrees=45, twistLimitDegrees=5}
]
