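# Post-process style cosmetic AnimBP for the F_MED_RecordScratch body that follows the parent mesh's pose and adds fur fixes and physics.
# The copied pose runs through deform and control rigs, then additive bone tweaks offset the backpack, plus fur bones during skydiving or open parachute.
# A RigidBody node simulates neck fur relative to neck_02, using world gravity recomputed each update by a deprecated gravity-override function.
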AnimBlueprint /BRCosmetics/Characters/Player/Female/Medium/Bodies/F_MED_RecordScratch/Meshes/F_MED_RecordScratch_AnimBP
target_skeleton=&/F_MED_RecordScratch_Skeleton
graphs [
 AnimGraph {nodes [
   n1 = Control Rig(Source=Copy Pose From Mesh() {bUseAttachedParent=true}) {
    alphaBoolBlend {blendOption=Linear}
    controlRigAssetReference {blueprintRigClass=&/F_MED_RecordScratch_DeformRig_Master_ControlRig.F_MED_RecordScratch_DeformRig_Master_ControlRig_C}
   }
   n2 = Control Rig!(Source=@n1, Alpha, translationAmount:float=Get CRtranslationAmount() -> (CRtranslationAmount:double)) -> (Pose) {
    alphaBoolBlend {blendOption=Linear}
    controlRigAssetReference {blueprintRigClass=&/F_MED_RecordScratch_CtrlRig.F_MED_RecordScratch_CtrlRig_C}
   }
   n3 = "Transform (Modify) Bone - Bone: attach_backpack"(ComponentPose=Local To Component(LocalPose=@n2), Alpha=1, Translation="0,3.5,0", Rotation="0,0,0", Scale="1,1,1") {
    translationMode=BMM_Additive
    translationSpace=BCS_BoneSpace
    alphaBoolBlend {blendOption=Linear}
    boneToModify {boneName=attach_backpack}
    translation {y=3.5}
   }
   n4 = Get bIsSkyDiving()
   n5 = "Transform (Modify) Bone - Bone: parent_bck_fur_push_r"!(ComponentPose=@n3, bAlphaBoolEnabled:bool=@n4, Translation="2,0,0", Rotation="0,0,0", Scale="1,1,1") -> (Pose) {
    alphaInputType=Bool
    translationMode=BMM_Additive
    translationSpace=BCS_BoneSpace
    alphaBoolBlend {blendOption=Linear}
    boneToModify {boneName=parent_bck_fur_push_r}
    translation {x=2}
   }
   n6 = "Transform (Modify) Bone - Bone: parent_bck_fur_push_l"!(ComponentPose=@n5, bAlphaBoolEnabled:bool=@n4, Translation="2,0,0", Rotation="0,0,0", Scale="1,1,1") -> (Pose) {
    alphaInputType=Bool
    translationMode=BMM_Additive
    translationSpace=BCS_BoneSpace
    alphaBoolBlend {blendOption=Linear}
    boneToModify {boneName=parent_bck_fur_push_l}
    translation {x=2}
   }
   n7 = Get bIsParachuteOpened()
   n8 = "Transform (Modify) Bone - Bone: base_shoulder_fur_l"!(ComponentPose=@n6, bAlphaBoolEnabled:bool=@n7, Translation="6.203,-1.070,-3.635", Rotation="-1.193,27.896,6.525", Scale="1,1,1") -> (Pose) {
    alphaInputType=Bool
    rotationMode=BMM_Additive
    rotationSpace=BCS_BoneSpace
    translationMode=BMM_Additive
    translationSpace=BCS_BoneSpace
    alphaBoolBlend {blendOption=Linear}
    boneToModify {boneName=base_shoulder_fur_l}
    rotation=<rot -1.193 27.896 6.525>
    translation=<6.203 -1.07 -3.635>
   }
   n9 = "Transform (Modify) Bone - Bone: base_shoulder_fur_r"!(ComponentPose=@n8, bAlphaBoolEnabled:bool=@n7, Translation="-3.470,-2.327,3.559", Rotation="3.347,40.006,-0.244", Scale="1,1,1") -> (Pose) {
    alphaInputType=Bool
    rotationMode=BMM_Additive
    rotationSpace=BCS_BoneSpace
    translationMode=BMM_Additive
    translationSpace=BCS_BoneSpace
    alphaBoolBlend {blendOption=Linear}
    boneToModify {boneName=base_shoulder_fur_r}
    rotation=<rot 3.347 40.006 -0.244>
    translation=<-3.4698 -2.3273 3.559>
   }
   n10 = RigidBody!(ComponentPose=@n9, Alpha, OverrideWorldGravity:Vector=Get Override World Gravity() -> (Override World Gravity:Vector), ExternalForce) -> (Pose) {
    bClampLinearTranslationLimitToRefPose=true
    bOverrideWorldGravity=true
    overridePhysicsAsset=&/F_MED_RecordScratch_Neckfur_Physics
    simulationSpace=BaseBoneSpace
    alphaBoolBlend {blendOption=Linear}
    baseBoneRef {boneName=neck_02}
    componentAppliedLinearAccClamp=<3000 3000 3000>
    componentLinearAccScale=<0.25 0.25 0.25>
    componentLinearVelScale=<0.43 0.43 0.43>
    simSpaceSettings {
     maxAngularAcceleration=1000
     maxAngularVelocity=3
     maxLinearAcceleration=700
     maxLinearVelocity=500
     worldAlpha=0.48
     externalLinearDragV=<0.2 0.2>
    }
   }
   Output Pose(Result=Component To Local(ComponentPose=@n10)) {layerGroup=None}
  ]}
 EventGraph {nodes [
   Event BlueprintUpdateAnimation()
   > n1 = Gravity Override State Function DEPRECATED(
    CharacterPartAnimInstance=Self-Reference()
    GameTimeInSeconds=Get Game Time in Seconds() -> (Game Time in Seconds:double)
    GravityOverrideParams=Get Gravity Override Params()
   )
   > Set Override World Gravity(Override World Gravity:Vector=@n1) -> (Output_Get:Vector)
  ]}
]
variables [
 CRtranslationAmount {type=number}
 Game Time in Seconds {type=number}
 Gravity Override Params {type=GravityOverrideParamsStruct}
 is targeting {type=boolean}
 Override World Gravity {type=Vector}
]
