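# Physics asset for the RecordScratch female medium character's FaceAcc mesh, providing collision bodies and joints for dynamic hair bones.
# A large head capsule anchors small capsules on back, side and front-tip hair bones, each constrained directly to the head with uniform angular limits.
# Left and right locks form three-body chains, with free-rotating bases on the head and mostly tight limits down the chain, except a wider swing on the last left joint.
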PhysicsAsset /BRCosmetics/Characters/Player/Female/Medium/Bodies/F_MED_RecordScratch/Meshes/Parts/F_MED_RecordScratch_FaceAcc_Physics
bodies {
 base_lock_hair_l [{shapeType=Capsule, length=0.5, radius=0.5}]
 base_lock_hair_r [{shapeType=Capsule, length=0.5, radius=0.5}]
 dyn_back_hair_a_m [{shapeType=Capsule, length=1, radius=1, center=<2.5 0 0>, rotation=<rot 90 0 0>}]
 dyn_back_hair_l [{shapeType=Capsule, length=1, radius=1, center=<2.5 0 0>, rotation=<rot 90 0 0>}]
 dyn_back_hair_low_r [{shapeType=Capsule, length=1, radius=1, center=<2.5 0 0>, rotation=<rot 90 0 0>}]
 dyn_back_hair_m [{shapeType=Capsule, length=1, radius=1, center=<2.5 0 0>, rotation=<rot 90 0 0>}]
 dyn_back_hair_r [{shapeType=Capsule, length=1, radius=1, center=<2.5 0 0>, rotation=<rot 90 0 0>}]
 dyn_back_hair_tip [{shapeType=Capsule, length=1, radius=1, center=<2.5 0 0>, rotation=<rot 90 0 0>}]
 dyn_back_hair_tip1 [{shapeType=Capsule, length=1, radius=1, center=<2.5 0 0>, rotation=<rot 90 0 0>}]
 dyn_back_hair_tip2 [{shapeType=Capsule, length=1, radius=1, center=<2.5 0 0>, rotation=<rot 90 0 0>}]
 dyn_back_hair_tip_low [{shapeType=Capsule, length=1, radius=1, center=<2.5 0 0>, rotation=<rot 90 0 0>}]
 dyn_back_hair_tip_m [{shapeType=Capsule, length=1, radius=1, center=<2.5 0 0>, rotation=<rot 90 0 0>}]
 dyn_hair_Tip_front [{shapeType=Capsule, length=1, radius=1, center=<2.5 0 0>, rotation=<rot 90 0 0>}]
 dyn_hair_Tip_front1 [{shapeType=Capsule, length=1, radius=1, center=<2.5 0 0>, rotation=<rot 90 0 0>}]
 dyn_hair_Tip_front2 [{shapeType=Capsule, length=1, radius=1, center=<2.5 0 0>, rotation=<rot 90 0 0>}]
 dyn_hair_side_l [{shapeType=Capsule, length=1, radius=1, center=<2.5 0 0>, rotation=<rot 90 0 0>}]
 dyn_hair_side_low_l [{shapeType=Capsule, length=1, radius=1, center=<2.5 0 0>, rotation=<rot 90 0 0>}]
 dyn_hair_side_r [{shapeType=Capsule, length=1, radius=1, center=<2.5 0 0>, rotation=<rot 90 0 0>}]
 dyn_lock_hair_01_l [{shapeType=Capsule, length=1, radius=1, shapeName=dyn_lock_hair_01_l_capsule, center=<1 0 0>, rotation=<rot 90 0 0>}]
 dyn_lock_hair_01_r [{shapeType=Capsule, length=1, radius=1, shapeName=dyn_lock_hair_01_r_capsule, center=<1 0 0>, rotation=<rot 90 0 0>}]
 dyn_lock_hair_02_l [{shapeType=Capsule, length=1, radius=1, shapeName=dyn_lock_hair_02_l_capsule, center=<1 0 0>, rotation=<rot 90 0 0>}]
 dyn_lock_hair_02_r [{shapeType=Capsule, length=1, radius=1, shapeName=dyn_lock_hair_02_r_capsule, center=<1 0 0>, rotation=<rot 90 0 0>}]
 head [{shapeType=Capsule, length=7.474, radius=7.6427, center=<1.7302 0.3051 0>, rotation=<rot 90 178.8065 168.8066>}]
}
constraints [
 {
  bone1Name=base_lock_hair_l
  bone2Name=head
  swing1LimitDegrees=45
  swing1Motion=Free
  swing2LimitDegrees=45
  swing2Motion=Free
  twistLimitDegrees=45
  twistMotion=Free
 }
 {
  bone1Name=base_lock_hair_r
  bone2Name=head
  swing1LimitDegrees=45
  swing1Motion=Free
  swing2LimitDegrees=45
  swing2Motion=Free
  twistLimitDegrees=45
  twistMotion=Free
 }
 {bone1Name=dyn_lock_hair_01_r, bone2Name=base_lock_hair_r, swing1LimitDegrees=4, swing2LimitDegrees=2, twistLimitDegrees=2}
 {bone1Name=dyn_lock_hair_02_r, bone2Name=dyn_lock_hair_01_r, swing1LimitDegrees=5, swing2LimitDegrees=5, twistLimitDegrees=2}
 {bone1Name=dyn_lock_hair_01_l, bone2Name=base_lock_hair_l, swing1LimitDegrees=5, swing2LimitDegrees=5, twistLimitDegrees=2}
 {bone1Name=dyn_lock_hair_02_l, bone2Name=dyn_lock_hair_01_l, swing1LimitDegrees=45, swing2LimitDegrees=2, twistLimitDegrees=2}
 {bone1Name=dyn_back_hair_a_m, bone2Name=head, swing1LimitDegrees=45, swing2LimitDegrees=45, twistLimitDegrees=45}
 {bone1Name=dyn_back_hair_l, bone2Name=head, swing1LimitDegrees=45, swing2LimitDegrees=45, twistLimitDegrees=45}
 {bone1Name=dyn_back_hair_low_r, bone2Name=head, swing1LimitDegrees=45, swing2LimitDegrees=45, twistLimitDegrees=45}
 {bone1Name=dyn_back_hair_m, bone2Name=head, swing1LimitDegrees=45, swing2LimitDegrees=45, twistLimitDegrees=45}
 {bone1Name=dyn_back_hair_r, bone2Name=head, swing1LimitDegrees=45, swing2LimitDegrees=45, twistLimitDegrees=45}
 {bone1Name=dyn_back_hair_tip, bone2Name=head, swing1LimitDegrees=45, swing2LimitDegrees=45, twistLimitDegrees=45}
 {bone1Name=dyn_back_hair_tip1, bone2Name=head, swing1LimitDegrees=45, swing2LimitDegrees=45, twistLimitDegrees=45}
 {bone1Name=dyn_back_hair_tip2, bone2Name=head, swing1LimitDegrees=45, swing2LimitDegrees=45, twistLimitDegrees=45}
 {bone1Name=dyn_back_hair_tip_low, bone2Name=head, swing1LimitDegrees=45, swing2LimitDegrees=45, twistLimitDegrees=45}
 {bone1Name=dyn_back_hair_tip_m, bone2Name=head, swing1LimitDegrees=45, swing2LimitDegrees=45, twistLimitDegrees=45}
 {bone1Name=dyn_hair_Tip_front, bone2Name=head, swing1LimitDegrees=45, swing2LimitDegrees=45, twistLimitDegrees=45}
 {bone1Name=dyn_hair_Tip_front1, bone2Name=head, swing1LimitDegrees=45, swing2LimitDegrees=45, twistLimitDegrees=45}
 {bone1Name=dyn_hair_Tip_front2, bone2Name=head, swing1LimitDegrees=45, swing2LimitDegrees=45, twistLimitDegrees=45}
 {bone1Name=dyn_hair_side_l, bone2Name=head, swing1LimitDegrees=45, swing2LimitDegrees=45, twistLimitDegrees=45}
 {bone1Name=dyn_hair_side_low_l, bone2Name=head, swing1LimitDegrees=45, swing2LimitDegrees=45, twistLimitDegrees=45}
 {bone1Name=dyn_hair_side_r, bone2Name=head, swing1LimitDegrees=45, swing2LimitDegrees=45, twistLimitDegrees=45}
]
